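# Niagara system for the WingBath main-player emote that draws a looping ring of flickering flame sprites.
# Flame001, the only enabled emitter with a renderer, bursts eight short-lived, world-up aligned sprites placed by execution index around a sphere location.
# A jagged curve scales their height over NormalizedLoopAge to fake flicker.
# The duplicate Flame, glow and smoke emitters are disabled or renderer-less; front-end flag and scale user parameters are exposed.
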NiagaraSystem /BRCosmetics/Animation/Game/MainPlayer/Emotes/WingBath/FX/NS_Emote_Wingbath_Flame_FX
dependencies {
 usedDataInterfaces [
  &NiagaraDataInterfaceCurve
  &NiagaraDataInterfaceVector2DCurve
  &NiagaraDataInterfaceSpriteRendererInfo
 ]
 usedDynamicInputs [
  &Multiply_Float
  &MakeCustomFloatFromBool
  &Vector2DFromFloat
 ]
 usedModules [&SystemState]
}
emitters {
 Center_Glow {rendererClasses [], renderers []}
 Flame {
  bEnabled=false
  emitterUpdateScript [
   &EmitterState(Life Cycle Mode=::NewEnumerator1, Loop Duration=0.3, UseLoopDelay=false)
   &SpawnBurst_Instantaneous(Spawn Count=8, Use Spawn Probability=false, UseLoopCountLimit=false)
  ]
  particleSpawnScript [
   &InitializeParticle(
    Color Mode=::NewEnumerator0
    Lifetime Max=0.31
    Lifetime Min=0.3
    Lifetime Mode=::NewEnumerator1
    Sprite Size Max=<15 50>
    Sprite Size Min=<13 45>
    Sprite Size Mode=::NewEnumerator2
    UsePositionOffset=false
    Write Material Random=true
   )
   /SetVariables(Particles.ExecutionIndex:NiagaraFloat=&ReturnNormalizedExecIndex(Normalized Index Includes One=false)) {isSetParameters=true, module=SetVariables_#1}
   &ShapeLocation(
    MaskSpawnGroup=false
    Offset=<0.1512 31.9221 6.2746>
    Offset Mode=::NewEnumerator0
    Rotation Mode=::NewEnumerator2
    Sphere Distribution=::NewEnumerator1
    Sphere Radius=16
    U Position=$Particles.ExecutionIndex
    UseOwnerScale=true
    Yaw / Pitch / Roll=<1.6833 0 0>
   )
   &SpriteFacingAndAlignment(Alignment Coordinate Space=::World, Sprite Alignment=<0 0 1>, WriteAlignment=true, WriteFacing=false) {module=SpriteFacingAndAlignment001}
  ]
  particleUpdateScript [
   &ParticleState()
   &ScaleSpriteSize(
    Non-Uniform Curve Index:NiagaraFloat=$System.NormalizedLoopAge
    Non-Uniform Curve Scale=&MakeVector2D(
     X=1
     Y=&FloatFromCurve(FloatCurve=DataInterface() {propertyValues {Curve {keys [<at 0 1>,<at 0.0985 0.8507>,<at 0.2008 1.0157>,<at 0.26 0.8488>,<at 0.3458 1.0027>,<at 0.3762 0.8539>,<at 0.4547 0.9889>,<at 0.5653 0.8593>,<at 0.6393 0.9746>,<at 0.7324 0.8661>,<at 0.8044 1.0027>,<at 0.9049 0.8638>,<at 0.9964 0.9311>]}}})
    )
    Non-Uniform Curve Sprite Scale=DataInterface() {propertyValues {
      XCurve {keys [{tangentMode=RCTM_User, time=0, value=1}, {tangentMode=RCTM_User, time=1, value=1}]}
      YCurve {keys [{tangentMode=RCTM_User, time=0, value=1}, {tangentMode=RCTM_User, time=1, value=1}]}
     }}
    Scale Sprite Size Mode=::NewEnumerator3
   )
  ]
  rendererClasses [&NiagaraSpriteRendererProperties]
  renderers [{rendererClass=&NiagaraSpriteRendererProperties}]
 }
 Flame001 {
  emitterUpdateScript [
   &EmitterState(Life Cycle Mode=::NewEnumerator1, Loop Duration=0.3, UseLoopDelay=false)
   &SpawnBurst_Instantaneous(Spawn Count=8, Use Spawn Probability=false, UseLoopCountLimit=false)
  ]
  particleSpawnScript [
   &InitializeParticle(
    Color Mode=::NewEnumerator0
    Lifetime Max=0.31
    Lifetime Min=0.3
    Lifetime Mode=::NewEnumerator1
    Sprite Size Max=<15 50>
    Sprite Size Min=<13 45>
    Sprite Size Mode=::NewEnumerator2
    UsePositionOffset=false
    Write Material Random=true
   )
   /SetVariables(Particles.ExecutionIndex:NiagaraFloat=&ReturnNormalizedExecIndex(Normalized Index Includes One=false)) {isSetParameters=true, module=SetVariables_#1}
   &ShapeLocation(
    MaskSpawnGroup=false
    Rotation Mode=::NewEnumerator2
    Sphere Distribution=::NewEnumerator1
    Sphere Radius=16
    U Position=$Particles.ExecutionIndex
    UseOwnerScale=true
    Yaw / Pitch / Roll=<1.6833 0 0>
   )
   &SpriteFacingAndAlignment(Alignment Coordinate Space=::World, Sprite Alignment=<0 0 1>, WriteAlignment=true, WriteFacing=false) {module=SpriteFacingAndAlignment001}
  ]
  particleUpdateScript [
   &ParticleState()
   &ScaleSpriteSize(
    Non-Uniform Curve Index:NiagaraFloat=$System.NormalizedLoopAge
    Non-Uniform Curve Scale=&MakeVector2D(
     X=1
     Y=&FloatFromCurve(FloatCurve=DataInterface() {propertyValues {Curve {keys [<at 0 1>,<at 0.0985 0.8507>,<at 0.2008 1.0157>,<at 0.26 0.8488>,<at 0.3458 1.0027>,<at 0.3762 0.8539>,<at 0.4547 0.9889>,<at 0.5653 0.8593>,<at 0.6393 0.9746>,<at 0.7324 0.8661>,<at 0.8044 1.0027>,<at 0.9049 0.8638>,<at 0.9964 0.9311>]}}})
    )
    Non-Uniform Curve Sprite Scale=DataInterface() {propertyValues {
      XCurve {keys [{tangentMode=RCTM_User, time=0, value=1}, {tangentMode=RCTM_User, time=1, value=1}]}
      YCurve {keys [{tangentMode=RCTM_User, time=0, value=1}, {tangentMode=RCTM_User, time=1, value=1}]}
     }}
    Scale Sprite Size Mode=::NewEnumerator3
   )
  ]
  rendererClasses [&NiagaraSpriteRendererProperties]
  renderers [{rendererClass=&NiagaraSpriteRendererProperties}]
 }
 Glow {
  bEnabled=false
  emitterUpdateScript [
   &EmitterState()
   &SpawnRate(SpawnRate=3, Use Spawn Probability=false)
  ]
  particleSpawnScript [
   &InitializeParticle(
    Color=<rgba 0.5814 1.1592 2 1>
    Lifetime=2
    Mass Max=2
    Mass Mode=::NewEnumerator3
    Sprite Size Mode=::NewEnumerator3
    Uniform Sprite Size=55
    UsePositionOffset=false
    Write Material Random=true
   )
   &SystemLocation(MaskSpawnGroup=false, Offset=<0 30.6379 3.4406>)
   /SetVariables(Particles.SpriteAlignment=<0 1 0>, Particles.SpriteFacing=<0 0 1>) {isSetParameters=true, module=SetVariables_#2}
  ]
  particleUpdateScript [
   &ParticleState()
   &ScaleColor(
    Scale Alpha=&FloatFromCurve(FloatCurve=DataInterface() {propertyValues {Curve {keys [
         {interpMode=RCIM_Cubic, tangentMode=RCTM_Break, time=0, value=0.6}
         {interpMode=RCIM_Cubic, tangentMode=RCTM_Break, time=0.2, value=1}
         {interpMode=RCIM_Cubic, tangentMode=RCTM_Break, time=1, value=0.6}
        ]}}})
    ScaleA=true
    ScaleRGB=true
   )
   &SolveForcesAndVelocity(Clamp Velocity=false, Limit Acceleration=false)
  ]
  rendererClasses [&NiagaraSpriteRendererProperties]
  renderers [{rendererClass=&NiagaraSpriteRendererProperties}]
 }
 Glow001 {
  bEnabled=false
  emitterUpdateScript [
   &EmitterState(Life Cycle Mode=::NewEnumerator1, Loop Behavior=::NewEnumerator1, Loop Delay=0.5, UseLoopDelay=false)
   &SpawnBurst_Instantaneous(Use Spawn Probability=false, UseLoopCountLimit=false)
  ]
  particleSpawnScript [
   &InitializeParticle(
    Color=<rgba 0.5814 1.1592 2 1>
    Lifetime=2
    Mass Max=2
    Mass Mode=::NewEnumerator3
    Sprite Size Mode=::NewEnumerator3
    Uniform Sprite Size=55
    UsePositionOffset=false
    Write Material Random=true
   )
   &SystemLocation(MaskSpawnGroup=false, Offset=<0 30.6379 3.4406>)
   /SetVariables(Particles.SpriteAlignment=<0 1 0>, Particles.SpriteFacing=<0 0 1>) {isSetParameters=true, module=SetVariables_#2}
  ]
  particleUpdateScript [
   &ParticleState()
   &ScaleColor(
    Scale Alpha=&FloatFromCurve(FloatCurve=DataInterface() {propertyValues {Curve {keys [
         {interpMode=RCIM_Cubic, tangentMode=RCTM_Break, time=0, value=0.6}
         {interpMode=RCIM_Cubic, tangentMode=RCTM_Break, time=0.2, value=1}
         {interpMode=RCIM_Cubic, tangentMode=RCTM_Break, time=1, value=0.6}
        ]}}})
    ScaleA=true
    ScaleRGB=true
   )
   &SolveForcesAndVelocity(Clamp Velocity=false, Limit Acceleration=false)
  ]
  rendererClasses [&NiagaraSpriteRendererProperties]
  renderers [{rendererClass=&NiagaraSpriteRendererProperties}]
 }
 Middle_Smoke_01 {rendererClasses [], renderers []}
 Middle_Smoke_02 {rendererClasses [], renderers []}
 Middle_Smoke_Camera {
  bEnabled=false
  emitterUpdateScript [
   &EmitterState()
   &SpawnRate(SpawnRate=15, Use Spawn Probability=false)
  ]
  particleSpawnScript [
   &InitializeParticle(
    Lifetime Max=0.5
    Lifetime Min=0.3
    Lifetime Mode=::NewEnumerator1
    Sprite Rotation Mode=::NewEnumerator2
    Sprite Size Mode=::NewEnumerator1
    Uniform Sprite Size Max=30
    Uniform Sprite Size Min=25
    UsePositionOffset=false
    Write Material Random=true
   )
   &ShapeLocation(
    MaskSpawnGroup=false
    Non Uniform Scale=<1 1 0.1>
    Offset=<-1.5948 31.0495 -6.9>
    Offset Mode=::NewEnumerator0
    Override Seed=false
    Sphere Radius=6.5
    UseOwnerScale=true
   )
   &ApplyInitialForces()
   &AddVelocity(Velocity=&RandomRangeVector(Minimum=<0 0 221>, Maximum=<0 0 247>))
  ]
  particleUpdateScript [
   &ParticleState()
   &ScaleSpriteSize(Uniform Curve Sprite Scale=DataInterface() {propertyValues {Curve {keys [
        {arriveTangent=0.5426, interpMode=RCIM_Cubic, leaveTangent=0.5426, tangentMode=RCTM_User, time=0, value=1}
        {interpMode=RCIM_Cubic, tangentMode=RCTM_User, time=1, value=1.2}
       ]}}})
   &ScaleColor(
    Scale Alpha=&FloatFromCurve(
     FloatCurve=DataInterface() {propertyValues {Curve {keys [
         {arriveTangent=-0.0366, interpMode=RCIM_Cubic, leaveTangent=-0.0366, tangentMode=RCTM_User, time=0, value=0}
         <at 0.3 1 RCIM_Cubic>
         <at 1 0 RCIM_Cubic>
        ]}}}
     Scale Curve=0.06
    )
    ScaleA=true
    ScaleRGB=true
   )
   &Drag(Drag=&RandomRangeFloat(Minimum=4, Maximum=5), Use Linear Drag=true, Use Rotational Drag=false)
   &SolveForcesAndVelocity(Clamp Velocity=false, Limit Acceleration=false)
   &DynamicMaterialParameters(
    Index 0 Param 1=&FloatFromCurve(FloatCurve=DataInterface() {propertyValues {Curve {keys [{interpMode=RCIM_Cubic, tangentMode=RCTM_Break, time=0.5, value=0.5}, {interpMode=RCIM_Cubic, tangentMode=RCTM_Break, time=1, value=-0.1}]}}})
    Index 0 Param 4=1
    Module.Index0Float0Write=true
    Module.Index0Float1Write=true
    Module.Index0Float2Write=true
    Module.Index0Float3Write=true
   ) {enabled=false}
   &SubUVAnimation(
    "Sprite Renderer "=DataInterface() {propertyValues {SpriteRenderer=&self:Middle_Smoke_Camera_0.NiagaraSpriteRendererProperties_1}}
    Start Frame Range Override=10
    UseEndFrame=true
    UseStartFrame=true
   )
   &CameraOffset(Camera Offset Amount=20) {enabled=false}
  ]
  rendererClasses [&NiagaraSpriteRendererProperties]
  renderers [{rendererClass=&NiagaraSpriteRendererProperties}]
 }
 Middle_Smoke_Camera001 {
  bEnabled=false
  emitterUpdateScript [
   &EmitterState()
   &SpawnRate(SpawnRate=15, Use Spawn Probability=false)
  ]
  particleSpawnScript [
   &InitializeParticle(
    Lifetime Max=0.5
    Lifetime Min=0.3
    Lifetime Mode=::NewEnumerator1
    Sprite Rotation Mode=::NewEnumerator2
    Sprite Size Mode=::NewEnumerator1
    Uniform Sprite Size Max=16
    Uniform Sprite Size Min=15
    UsePositionOffset=false
    Write Material Random=true
   )
   &ShapeLocation(
    Cylinder Height=0
    Cylinder Radius=10
    MaskSpawnGroup=false
    Non Uniform Scale=<1 1 0.1>
    Offset=<-1.5948 31.0495 -6.9>
    Offset Mode=::NewEnumerator0
    Override Seed=false
    Shape Primitive=::NewEnumerator1
    Surface Only=true
    Use Endcaps In Surface Only Mode=false
    UseOwnerScale=true
   )
   &ApplyInitialForces()
   &AddVelocity(Velocity=&RandomRangeVector(Minimum=<0 0 221>, Maximum=<0 0 247>))
   &AddVelocity(Velocity=<0 0 90>) {enabled=false, module=AddVelocity001}
  ]
  particleUpdateScript [
   &ParticleState()
   &ScaleSpriteSize(Uniform Curve Sprite Scale=DataInterface() {propertyValues {Curve {keys [
        {arriveTangent=0.5426, interpMode=RCIM_Cubic, leaveTangent=0.5426, tangentMode=RCTM_User, time=0, value=1}
        {interpMode=RCIM_Cubic, tangentMode=RCTM_User, time=1, value=1.2}
       ]}}})
   &ScaleColor(
    Scale Alpha=&FloatFromCurve(
     FloatCurve=DataInterface() {propertyValues {Curve {keys [
         {arriveTangent=-0.0366, interpMode=RCIM_Cubic, leaveTangent=-0.0366, tangentMode=RCTM_User, time=0, value=0}
         <at 0.3 1 RCIM_Cubic>
         <at 1 0 RCIM_Cubic>
        ]}}}
     Scale Curve=0.06
    )
    ScaleA=true
    ScaleRGB=true
   )
   &Drag(Drag=&RandomRangeFloat(Minimum=9, Maximum=10), Use Linear Drag=true, Use Rotational Drag=false)
   &SolveForcesAndVelocity(Clamp Velocity=false, Limit Acceleration=false)
   &DynamicMaterialParameters(
    Index 0 Param 1=&FloatFromCurve(FloatCurve=DataInterface() {propertyValues {Curve {keys [{interpMode=RCIM_Cubic, tangentMode=RCTM_Break, time=0.5, value=0.5}, {interpMode=RCIM_Cubic, tangentMode=RCTM_Break, time=1, value=-0.1}]}}})
    Index 0 Param 4=1
    Module.Index0Float0Write=true
    Module.Index0Float1Write=true
    Module.Index0Float2Write=true
    Module.Index0Float3Write=true
   ) {enabled=false}
   &SubUVAnimation(
    "Sprite Renderer "=DataInterface() {propertyValues {SpriteRenderer=&self:Middle_Smoke_Camera001_0.NiagaraSpriteRendererProperties_1}}
    Start Frame Range Override=10
    UseEndFrame=true
    UseStartFrame=true
   )
   &CameraOffset(Camera Offset Amount=20) {enabled=false}
  ]
  rendererClasses [&NiagaraSpriteRendererProperties]
  renderers [{rendererClass=&NiagaraSpriteRendererProperties}]
 }
}
summary {
 systemName=NS_Emote_Wingbath_Flame_FX
 emitters [
  {bEnabled=false, emitterName=Flame, rendererClasses [&NiagaraSpriteRendererProperties]}
  {bEnabled=false, emitterName=Middle_Smoke_Camera, rendererClasses [&NiagaraSpriteRendererProperties]}
  {bEnabled=false, emitterName=Glow, rendererClasses [&NiagaraSpriteRendererProperties]}
  {bEnabled=false, emitterName=Glow001, rendererClasses [&NiagaraSpriteRendererProperties]}
  {bEnabled=false, emitterName=Middle_Smoke_Camera001, rendererClasses [&NiagaraSpriteRendererProperties]}
  {emitterName=Middle_Smoke_01, rendererClasses []}
  {emitterName=Middle_Smoke_02, rendererClasses []}
  {emitterName=Center_Glow, rendererClasses []}
  {emitterName=Flame001, rendererClasses [&NiagaraSpriteRendererProperties]}
 ]
 userVariables [
  User.bIsFrontEnd {type=NiagaraBool}
  User.FrontEndScale {type=NiagaraFloat, defaultValue=0.6, description=""}
 ]
}
