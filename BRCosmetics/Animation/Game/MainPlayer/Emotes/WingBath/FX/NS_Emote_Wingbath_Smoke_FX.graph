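# Niagara system for the smoke effect of the main player's Wingbath emote, using one sprite emitter timed to 8.4 seconds.
# Faint puffs spawn continuously on a flat cylinder, rise under upward acceleration against drag, grow, and fade.
# Per-particle randoms and age curves drive dynamic material parameters and decelerating sprite spin, with a camera offset applied.
# The burst spawn is disabled, and no shown module reads the front-end user variables.
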NiagaraSystem /BRCosmetics/Animation/Game/MainPlayer/Emotes/WingBath/FX/NS_Emote_Wingbath_Smoke_FX
dependencies {
 usedDataInterfaces [&NiagaraDataInterfaceCurve]
 usedDynamicInputs [&SimulationPosition, &MakeFloatFromVector4]
 usedModules [&SystemState]
}
emitters {Smoke {
  emitterUpdateScript [
   &EmitterState(Life Cycle Mode=::NewEnumerator1, Loop Behavior=::NewEnumerator1, Loop Duration=8.4, UseLoopDelay=false)
   &SpawnBurst_Instantaneous(Spawn Count=3, Use Spawn Probability=false, UseLoopCountLimit=false) {enabled=false}
   &SpawnRate(SpawnRate=15, Use Spawn Probability=false)
  ]
  particleSpawnScript [
   &InitializeParticle(
    Color=<rgba 1 1 1 0.06>
    Lifetime=&UniformRangedFloat(Minimum=1.1, Maximum=1.5)
    Sprite Size Mode=::NewEnumerator1
    Uniform Sprite Size Max=14
    Uniform Sprite Size Min=12
    UsePositionOffset=false
    Write Material Random=true
   ) {module=InitializeParticle001}
   /SetVariables(Particles.RandomFloat=&UniformRangedFloat(Minimum=0, Maximum=1)) {isSetParameters=true}
   &AddVelocity(Velocity=/Niagara/DynamicInputs/Random/RandomVector(
     Vector Scale:NiagaraFloat=25
     Constraint Vector:Vector3f=<0 0 1>
     Randomness Mode:ENiagaraRandomnessMode=::NewEnumerator0
     Random Seed:NiagaraInt32=0
     Fixed Random Seed:NiagaraBool=false
    ))
   &CylinderLocation(
    Cylinder Height=0
    Cylinder Height Midpoint=1
    Cylinder Radius=5
    MaskSpawnGroup=false
    Offset Mode:ENiagara_OffsetMode=::NewEnumerator0
    Override Seed=false
    UseOwnerScale=true
   )
  ]
  particleUpdateScript [
   &ParticleState()
   &AccelerationForce(Acceleration=&UniformRangedVector(Minimum=<0 0 250>, Maximum=<0 0 300>))
   &Drag(Drag=4, Use Linear Drag=true, Use Rotational Drag=false)
   &DynamicMaterialParameters(
    Index 0 Param 1=&Add_Float(
     A=$Particles.RandomFloat
     B=&FloatFromCurve(FloatCurve=DataInterface() {propertyValues {Curve {keys [
          {arriveTangent=2.4104, time=0, value=1}
          {arriveTangent=0.2537, interpMode=RCIM_Cubic, leaveTangent=0.2537, tangentMode=RCTM_User, time=1, value=2}
         ]}}})
    )
    Index 0 Param 4=1
    Index 2 Param 1=&FloatFromCurve(FloatCurve=DataInterface() {propertyValues {Curve {keys [{interpMode=RCIM_Cubic, leaveTangent=0.9717, tangentMode=RCTM_Break, time=0.25, value=0.8}, <at 1 1>]}}})
    Index 2 Param 2=&FloatFromCurve(
     FloatCurve=DataInterface() {propertyValues {Curve {keys [<at 0 0.5 RCIM_Cubic>,<at 1 0 RCIM_Cubic>]}}}
     Scale Curve=&RandomRangeFloat(Minimum=0.2, Maximum=0.5)
    )
    Index 2 Param 3=&FloatFromCurve(FloatCurve=DataInterface() {propertyValues {Curve {keys [
         {interpMode=RCIM_Cubic, tangentMode=RCTM_Break, time=0, value=1}
         {interpMode=RCIM_Cubic, leaveTangent=11.1233, tangentMode=RCTM_Break, time=1, value=3}
        ]}}})
    Index 2 Param 4=&FloatFromCurve(
     FloatCurve=DataInterface() {propertyValues {Curve {keys [
         {arriveTangent=-0.4552, interpMode=RCIM_Cubic, leaveTangent=-0.4552, tangentMode=RCTM_User, time=0, value=1}
         {interpMode=RCIM_Cubic, leaveTangent=-1.2956, tangentMode=RCTM_Break, time=1, value=0.75}
        ]}}}
     Scale Curve=1.4
    )
    Module.Index0Float0Write=true
    Module.Index0Float1Write=true
    Module.Index0Float2Write=true
    Module.Index0Float3Write=true
    Module.Index2Float0Write=true
    Module.Index2Float1Write=true
    Module.Index2Float2Write=true
    Module.Index2Float3Write=true
    Write Parameter Index 2=::NewEnumerator1
   ) {module=DynamicMaterialParameters001}
   &ScaleSpriteSize(Scale Factor=&Vector2DFromFloat(Value=&FloatFromCurve(
      FloatCurve=DataInterface() {propertyValues {Curve {keys [
          {arriveTangent=3.358, interpMode=RCIM_Cubic, leaveTangent=3.358, tangentMode=RCTM_User, time=0, value=0.1773}
          {arriveTangent=0.8227, interpMode=RCIM_Cubic, leaveTangent=0.8227, time=0.1179, value=0.489}
          {interpMode=RCIM_Cubic, leaveTangent=0.0321, time=1, value=1}
         ]}}}
      Scale Curve=2
     )))
   &ScaleColor(
    Scale Alpha=&FloatFromCurve(FloatCurve=DataInterface() {propertyValues {Curve {keys [
         {arriveTangent=9191.0225, interpMode=RCIM_Cubic, leaveTangent=4.1064, tangentMode=RCTM_User, time=0.0001, value=0.0123}
         {arriveTangentWeight=0.0484, interpMode=RCIM_Cubic, leaveTangentWeight=0.2906, tangentMode=RCTM_User, time=0.1, value=1}
         {arriveTangent=-1.1111, interpMode=RCIM_Cubic, leaveTangent=-1.1111, time=0.8, value=0.75}
         <at 1 0 RCIM_Cubic>
        ]}}})
    ScaleA=true
    ScaleRGB=true
   )
   &SpriteRotationRate(Rotation Rate=&Multiply_Float(
     A=&Lerp_Float(A=-150, B=150, Alpha=$Particles.MaterialRandom)
     B=&Lerp_Float(A=1, B=0, Alpha=&Power(a=$Particles.NormalizedAge, b=0.7))
    ))
   &CameraOffset(Camera Offset Amount=21)
   &SolveForcesAndVelocity(Clamp Velocity=false, Limit Acceleration=false)
  ]
  rendererClasses [&NiagaraSpriteRendererProperties]
  renderers [{rendererClass=&NiagaraSpriteRendererProperties}]
 }}
summary {
 systemName=NS_Emote_Wingbath_Smoke_FX
 emitters [{emitterName=Smoke, rendererClasses [&NiagaraSpriteRendererProperties]}]
 userVariables [
  User.bIsFrontEnd {type=NiagaraBool}
  User.FrontEndScale {type=NiagaraFloat, defaultValue=0.6, description=""}
 ]
}
