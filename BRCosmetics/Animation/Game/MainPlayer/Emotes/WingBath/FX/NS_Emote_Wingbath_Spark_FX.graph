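# Niagara spark effect for the MainPlayer WingBath emote.
# Enabled emitters cover small and large sparks, flares and a light, though their modules and renderers are not listed here.
# The disabled Flame emitters burst index-distributed sprites, using curves on system loop age to flicker size and pulse brightness.
# Other disabled spark and flare emitters remain, and user parameters expose a front-end flag and scale.
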NiagaraSystem /BRCosmetics/Animation/Game/MainPlayer/Emotes/WingBath/FX/NS_Emote_Wingbath_Spark_FX
dependencies {
 usedDataInterfaces [&NiagaraDataInterfaceCurve, &NiagaraDataInterfaceVector2DCurve]
 usedDynamicInputs []
 usedModules [&SystemState]
}
emitters {
 Flame {
  bEnabled=false
  emitterUpdateScript [
   &EmitterState(Life Cycle Mode=::NewEnumerator1, Loop Duration=0.15, UseLoopDelay=false)
   &SpawnBurst_Instantaneous(Spawn Count=16, Use Spawn Probability=false, UseLoopCountLimit=false)
  ]
  particleSpawnScript [
   &InitializeParticle(
    Color Mode=::NewEnumerator0
    Lifetime=5
    Sprite Size Max=<10 42>
    Sprite Size Min=<7 32>
    Sprite Size Mode=::NewEnumerator2
    UsePositionOffset=false
    Write Material Random=true
   )
   /SetVariables(Particles.ExecutionIndex:NiagaraFloat=&ReturnNormalizedExecIndex(Normalized Index Includes One=false)) {isSetParameters=true, module=SetVariables_#1}
   &ShapeLocation(
    MaskSpawnGroup=false
    Offset=<0.1512 31.9221 6.2746>
    Offset Mode=::NewEnumerator0
    Rotation Mode=::NewEnumerator2
    Sphere Distribution=::NewEnumerator1
    Sphere Radius=16
    U Position=$Particles.ExecutionIndex
    UseOwnerScale=true
    Yaw / Pitch / Roll=<1.6833 0 0>
   )
   &SpriteFacingAndAlignment(Alignment Coordinate Space=::World, Sprite Alignment=<0 0 1>, WriteAlignment=true, WriteFacing=false) {module=SpriteFacingAndAlignment001}
  ]
  particleUpdateScript [
   &ParticleState()
   &ScaleColor(
    Scale RGB=&VectorFromFloat(Value=&FloatFromCurve(
      FloatCurve=DataInterface() {propertyValues {Curve {keys [
          {interpMode=RCIM_Cubic, leaveTangent=-22.1259, tangentMode=RCTM_Break, time=0.0633, value=3}
          {interpMode=RCIM_Cubic, tangentMode=RCTM_Break, time=0.11, value=2.1708}
          <at 0.5 1.5>
         ]}}}
      Scale Curve=&FloatFromCurve(
       FloatCurve=DataInterface() {propertyValues {Curve {keys [
           {interpMode=RCIM_Cubic, leaveTangent=-10.6544, tangentMode=RCTM_Break, time=0, value=2}
           {interpMode=RCIM_Cubic, tangentMode=RCTM_Break, time=0.1, value=1}
           {interpMode=RCIM_Cubic, tangentMode=RCTM_Break, time=0.1768, value=1}
           {interpMode=RCIM_Cubic, leaveTangent=-10.4556, tangentMode=RCTM_Break, time=0.1843, value=2}
           {interpMode=RCIM_Cubic, tangentMode=RCTM_Break, time=0.2894, value=1}
           {interpMode=RCIM_Cubic, tangentMode=RCTM_Break, time=0.3535, value=1}
           {interpMode=RCIM_Cubic, leaveTangent=-9.6474, tangentMode=RCTM_Break, time=0.3611, value=2}
           {interpMode=RCIM_Cubic, tangentMode=RCTM_Break, time=0.4661, value=1}
           {interpMode=RCIM_Cubic, tangentMode=RCTM_Break, time=1, value=1}
          ]}}}
       CurveIndex=$System.NormalizedLoopAge
      )
     ))
    ScaleA=true
    ScaleRGB=true
   )
   &ScaleSpriteSize(
    Uniform Curve Index=$System.NormalizedLoopAge
    Uniform Curve Sprite Scale=DataInterface() {propertyValues {Curve {keys [
        {interpMode=RCIM_Cubic, leaveTangent=-100.3098, tangentMode=RCTM_Break, time=0.985, value=1}
        {interpMode=RCIM_Cubic, tangentMode=RCTM_Break, time=1, value=0}
       ]}}}
   ) {enabled=false, module=ScaleSpriteSize001}
   &ScaleSpriteSize(
    Non-Uniform Curve Index:NiagaraFloat=$System.NormalizedLoopAge
    Non-Uniform Curve Scale=&MakeVector2D(
     X=1
     Y=&FloatFromCurve(FloatCurve=DataInterface() {propertyValues {Curve {keys [<at 0 1>,<at 0.1429 0.5454>,<at 0.2157 1.03>,<at 0.3073 0.5526>,<at 0.3469 1.0125>,<at 0.4429 0.5454>,<at 0.4558 1.0118>,<at 0.5429 0.5454>,<at 0.5431 1.0118>,<at 0.6429 0.5454>,<at 0.666 0.9973>,<at 0.7429 0.5954>,<at 0.7622 0.9973>,<at 0.8429 0.5954>,<at 0.8691 1.0082>,<at 0.9429 0.5454>,<at 0.9511 1.0118>,<at 0.9964 0.9311>]}}})
    )
    Non-Uniform Curve Sprite Scale=DataInterface() {propertyValues {
      XCurve {
       defaultValue=1.85
       keys [
        {arriveTangent=-85.7451, interpMode=RCIM_Cubic, leaveTangent=-14.2535, leaveTangentWeight=0.3283, tangentMode=RCTM_Break, time=0, value=1.7}
        {
         arriveTangent=-0.1408
         arriveTangentWeight=0.04
         interpMode=RCIM_Cubic
         leaveTangent=0.1534
         leaveTangentWeight=0.2883
         tangentMode=RCTM_Break
         time=0.1
         value=1
        }
        {
         arriveTangent=0.094
         arriveTangentWeight=0.0189
         interpMode=RCIM_Cubic
         leaveTangent=150.2723
         leaveTangentWeight=0.0025
         tangentMode=RCTM_Break
         time=0.1768
         value=1
        }
        {
         arriveTangent=0.0001
         arriveTangentWeight=0.0214
         interpMode=RCIM_Cubic
         leaveTangent=-15.5273
         leaveTangentWeight=0.2669
         tangentMode=RCTM_Break
         time=0.1843
         value=1.7
        }
        {
         arriveTangent=-0.0955
         arriveTangentWeight=0.0383
         interpMode=RCIM_Cubic
         leaveTangent=-0.0955
         leaveTangentWeight=0.2286
         tangentMode=RCTM_Break
         time=0.2894
         value=1
        }
        {
         arriveTangent=0.0961
         arriveTangentWeight=0.0181
         interpMode=RCIM_Cubic
         leaveTangent=135.621
         leaveTangentWeight=0.2105
         tangentMode=RCTM_Break
         time=0.3535
         value=1
        }
        {
         arriveTangentWeight=0.0025
         interpMode=RCIM_Cubic
         leaveTangent=-18.5918
         leaveTangentWeight=0.208
         tangentMode=RCTM_Break
         time=0.3611
         value=1.7
        }
        {
         arriveTangent=-1.3624
         arriveTangentWeight=0.0383
         interpMode=RCIM_Cubic
         leaveTangent=-0.0605
         leaveTangentWeight=0.1696
         tangentMode=RCTM_Break
         time=0.4661
         value=1
        }
        {interpMode=RCIM_Cubic, leaveTangent=-138.3839, tangentMode=RCTM_Break, time=0.985, value=1}
        {arriveTangentWeight=0.005, interpMode=RCIM_Cubic, tangentMode=RCTM_Break, time=1, value=0}
       ]
      }
      YCurve {
       defaultValue=1
       keys [
        {arriveTangent=-85.7451, interpMode=RCIM_Cubic, leaveTangent=-11.9958, leaveTangentWeight=0.3283, tangentMode=RCTM_Break, time=0, value=1.85}
        {
         arriveTangent=-0.1408
         arriveTangentWeight=0.0067
         interpMode=RCIM_Cubic
         leaveTangent=0.1534
         leaveTangentWeight=0.3217
         tangentMode=RCTM_Break
         time=0.12
         value=1
        }
        {
         arriveTangent=0.094
         arriveTangentWeight=0.0523
         interpMode=RCIM_Cubic
         leaveTangent=150.2723
         leaveTangentWeight=0.0025
         tangentMode=RCTM_Break
         time=0.1768
         value=1
        }
        {
         arriveTangent=0.0001
         arriveTangentWeight=0.0548
         interpMode=RCIM_Cubic
         leaveTangent=-12.1025
         leaveTangentWeight=0.2669
         tangentMode=RCTM_Break
         time=0.1843
         value=1.85
        }
        {
         arriveTangent=-0.0955
         arriveTangentWeight=0.005
         interpMode=RCIM_Cubic
         leaveTangent=-0.0955
         leaveTangentWeight=0.2619
         tangentMode=RCTM_Break
         time=0.2994
         value=1
        }
        {
         arriveTangent=0.0961
         arriveTangentWeight=0.0514
         interpMode=RCIM_Cubic
         leaveTangent=135.621
         leaveTangentWeight=0.2105
         tangentMode=RCTM_Break
         time=0.3535
         value=1
        }
        {
         arriveTangentWeight=0.0025
         interpMode=RCIM_Cubic
         leaveTangent=-16.9812
         leaveTangentWeight=0.208
         tangentMode=RCTM_Break
         time=0.3611
         value=1.85
        }
        {
         arriveTangent=-1.3624
         arriveTangentWeight=0.005
         interpMode=RCIM_Cubic
         leaveTangent=-0.0605
         leaveTangentWeight=0.203
         tangentMode=RCTM_Break
         time=0.4761
         value=1
        }
        {interpMode=RCIM_Cubic, leaveTangent=-138.3839, tangentMode=RCTM_Break, time=0.985, value=1}
        {arriveTangentWeight=0.005, interpMode=RCIM_Cubic, tangentMode=RCTM_Break, time=1, value=0}
       ]
      }
     }}
    Scale Sprite Size Mode=::NewEnumerator3
   ) {module=ScaleSpriteSize003}
  ]
  rendererClasses [&NiagaraSpriteRendererProperties]
  renderers [{rendererClass=&NiagaraSpriteRendererProperties}]
 }
 Flame002 {
  bEnabled=false
  emitterUpdateScript [
   &EmitterState(Life Cycle Mode=::NewEnumerator1, Loop Duration=0.15, UseLoopDelay=false)
   &SpawnBurst_Instantaneous(Spawn Count=16, Use Spawn Probability=false, UseLoopCountLimit=false)
  ]
  particleSpawnScript [
   &InitializeParticle(
    Color Mode=::NewEnumerator0
    Lifetime Max=0.3
    Lifetime Min=0.27
    Lifetime Mode=::NewEnumerator1
    Sprite Rotation Angle=-90
    Sprite Rotation Mode=::NewEnumerator0
    Sprite Size Max=<48 6.048>
    Sprite Size Min=<46 5.5>
    Sprite Size Mode=::NewEnumerator2
    UsePositionOffset=false
    Write Material Random=true
   )
   /SetVariables(Particles.ExecutionIndex:NiagaraFloat=&ReturnNormalizedExecIndex(Normalized Index Includes One=false)) {isSetParameters=true, module=SetVariables_#1}
   &ShapeLocation(
    MaskSpawnGroup=false
    Offset=<0.1512 31.9221 3.3462>
    Offset Mode=::NewEnumerator0
    Rotation Mode=::NewEnumerator2
    Sphere Distribution=::NewEnumerator1
    Sphere Radius=16
    U Position=$Particles.ExecutionIndex
    UseOwnerScale=true
    Yaw / Pitch / Roll=<1.6833 0 0>
   )
   &SpriteFacingAndAlignment(Alignment Coordinate Space=::World, Sprite Alignment=<0 0 1>, WriteAlignment=true, WriteFacing=false) {module=SpriteFacingAndAlignment001}
  ]
  particleUpdateScript [
   &ParticleState()
   &ScaleSpriteSize(
    Non-Uniform Curve Sprite Scale=DataInterface() {propertyValues {
      XCurve {keys [
        {interpMode=RCIM_Cubic, leaveTangent=0.3104, tangentMode=RCTM_Break, time=0, value=0.8}
        {interpMode=RCIM_Cubic, tangentMode=RCTM_Break, time=1, value=1}
       ]}
      YCurve {keys [
        {arriveTangent=0.5062, interpMode=RCIM_Cubic, leaveTangent=0.5062, tangentMode=RCTM_User, time=0, value=0.7}
        {interpMode=RCIM_Cubic, tangentMode=RCTM_Break, time=1, value=1}
       ]}
     }}
    Scale Sprite Size Mode=::NewEnumerator3
   )
   &ScaleSpriteSize(
    Uniform Curve Index=$System.NormalizedLoopAge
    Uniform Curve Sprite Scale=DataInterface() {propertyValues {Curve {keys [
        {interpMode=RCIM_Cubic, leaveTangent=-100.3098, tangentMode=RCTM_Break, time=0.985, value=1}
        {interpMode=RCIM_Cubic, tangentMode=RCTM_Break, time=1, value=0}
       ]}}}
   ) {enabled=false, module=ScaleSpriteSize001}
   &ScaleSpriteSize(
    Non-Uniform Curve Index:NiagaraFloat=$System.NormalizedLoopAge
    Non-Uniform Curve Sprite Scale=DataInterface() {propertyValues {
      XCurve {
       defaultValue=1.85
       keys [
        {arriveTangent=-85.7451, interpMode=RCIM_Cubic, leaveTangent=-14.2535, leaveTangentWeight=0.3283, tangentMode=RCTM_Break, time=0, value=1.7}
        {
         arriveTangent=-0.1408
         arriveTangentWeight=0.04
         interpMode=RCIM_Cubic
         leaveTangent=0.1534
         leaveTangentWeight=0.2883
         tangentMode=RCTM_Break
         time=0.1
         value=1
        }
        {
         arriveTangent=0.094
         arriveTangentWeight=0.0189
         interpMode=RCIM_Cubic
         leaveTangent=150.2723
         leaveTangentWeight=0.0025
         tangentMode=RCTM_Break
         time=0.1768
         value=1
        }
        {
         arriveTangent=0.0001
         arriveTangentWeight=0.0214
         interpMode=RCIM_Cubic
         leaveTangent=-15.5273
         leaveTangentWeight=0.2669
         tangentMode=RCTM_Break
         time=0.1843
         value=1.7
        }
        {
         arriveTangent=-0.0955
         arriveTangentWeight=0.0383
         interpMode=RCIM_Cubic
         leaveTangent=-0.0955
         leaveTangentWeight=0.2285
         tangentMode=RCTM_Break
         time=0.2894
         value=1
        }
        {
         arriveTangent=0.0961
         arriveTangentWeight=0.0181
         interpMode=RCIM_Cubic
         leaveTangent=135.621
         leaveTangentWeight=0.2105
         tangentMode=RCTM_Break
         time=0.3535
         value=1
        }
        {
         arriveTangentWeight=0.0025
         interpMode=RCIM_Cubic
         leaveTangent=-18.5918
         leaveTangentWeight=0.208
         tangentMode=RCTM_Break
         time=0.3611
         value=1.7
        }
        {
         arriveTangent=-1.3624
         arriveTangentWeight=0.0383
         interpMode=RCIM_Cubic
         leaveTangent=-0.0605
         leaveTangentWeight=0.1696
         tangentMode=RCTM_Break
         time=0.4661
         value=1
        }
        {interpMode=RCIM_Cubic, leaveTangent=-138.3839, tangentMode=RCTM_Break, time=0.985, value=1}
        {arriveTangentWeight=0.005, interpMode=RCIM_Cubic, tangentMode=RCTM_Break, time=1, value=0}
       ]
      }
      YCurve {
       defaultValue=1
       keys [
        {arriveTangent=-85.7451, interpMode=RCIM_Cubic, leaveTangent=-11.9958, leaveTangentWeight=0.3283, tangentMode=RCTM_Break, time=0, value=1.85}
        {
         arriveTangent=-0.1408
         arriveTangentWeight=0.0067
         interpMode=RCIM_Cubic
         leaveTangent=0.1534
         leaveTangentWeight=0.3217
         tangentMode=RCTM_Break
         time=0.12
         value=1
        }
        {
         arriveTangent=0.094
         arriveTangentWeight=0.0523
         interpMode=RCIM_Cubic
         leaveTangent=150.2723
         leaveTangentWeight=0.0025
         tangentMode=RCTM_Break
         time=0.1768
         value=1
        }
        {
         arriveTangent=0.0001
         arriveTangentWeight=0.0548
         interpMode=RCIM_Cubic
         leaveTangent=-12.1025
         leaveTangentWeight=0.2669
         tangentMode=RCTM_Break
         time=0.1843
         value=1.85
        }
        {
         arriveTangent=-0.0955
         arriveTangentWeight=0.005
         interpMode=RCIM_Cubic
         leaveTangent=-0.0955
         leaveTangentWeight=0.2619
         tangentMode=RCTM_Break
         time=0.2994
         value=1
        }
        {
         arriveTangent=0.0961
         arriveTangentWeight=0.0514
         interpMode=RCIM_Cubic
         leaveTangent=135.621
         leaveTangentWeight=0.2105
         tangentMode=RCTM_Break
         time=0.3535
         value=1
        }
        {
         arriveTangentWeight=0.0025
         interpMode=RCIM_Cubic
         leaveTangent=-16.9812
         leaveTangentWeight=0.208
         tangentMode=RCTM_Break
         time=0.3611
         value=1.85
        }
        {
         arriveTangent=-1.3624
         arriveTangentWeight=0.005
         interpMode=RCIM_Cubic
         leaveTangent=-0.0605
         leaveTangentWeight=0.203
         tangentMode=RCTM_Break
         time=0.4761
         value=1
        }
        {interpMode=RCIM_Cubic, leaveTangent=-138.3839, tangentMode=RCTM_Break, time=0.985, value=1}
        {arriveTangentWeight=0.005, interpMode=RCIM_Cubic, tangentMode=RCTM_Break, time=1, value=0}
       ]
      }
     }}
    Scale Sprite Size Mode=::NewEnumerator3
   ) {enabled=false, module=ScaleSpriteSize002}
   &ScaleSpriteSize(
    Non-Uniform Curve Index:NiagaraFloat=$System.NormalizedLoopAge
    Non-Uniform Curve Scale=&MakeVector2D(
     X=&FloatFromCurve(FloatCurve=DataInterface() {propertyValues {Curve {keys [<at 0 1>,<at 0.0771 0.8762>,<at 0.1405 1>,<at 0.1708 0.884>,<at 0.2278 1.0181>,<at 0.2509 0.884>,<at 0.2986 1.0403>,<at 0.3293 0.8803>,<at 0.3846 1.0072>,<at 0.4148 0.884>,<at 0.4558 0.9964>,<at 0.5012 0.8948>,<at 0.5431 0.9964>,<at 0.6126 0.8803>,<at 0.666 0.9819>,<at 0.7161 0.9193>,<at 0.7622 0.9819>,<at 0.8196 0.9138>,<at 0.8691 0.9927>,<at 0.9065 0.8948>,<at 0.9511 0.9964>,<at 0.9964 0.9311>]}}})
     Y=1
    )
    Non-Uniform Curve Sprite Scale=DataInterface() {propertyValues {
      XCurve {
       defaultValue=1.85
       keys [
        {arriveTangent=-85.7451, interpMode=RCIM_Cubic, leaveTangent=-14.2535, leaveTangentWeight=0.3283, tangentMode=RCTM_Break, time=0, value=1.7}
        {
         arriveTangent=-0.1408
         arriveTangentWeight=0.04
         interpMode=RCIM_Cubic
         leaveTangent=0.1534
         leaveTangentWeight=0.2883
         tangentMode=RCTM_Break
         time=0.1
         value=1
        }
        {
         arriveTangent=0.094
         arriveTangentWeight=0.0189
         interpMode=RCIM_Cubic
         leaveTangent=150.2723
         leaveTangentWeight=0.0025
         tangentMode=RCTM_Break
         time=0.1768
         value=1
        }
        {
         arriveTangent=0.0001
         arriveTangentWeight=0.0214
         interpMode=RCIM_Cubic
         leaveTangent=-15.5273
         leaveTangentWeight=0.2669
         tangentMode=RCTM_Break
         time=0.1843
         value=1.7
        }
        {
         arriveTangent=-0.0955
         arriveTangentWeight=0.0383
         interpMode=RCIM_Cubic
         leaveTangent=-0.0955
         leaveTangentWeight=0.2286
         tangentMode=RCTM_Break
         time=0.2894
         value=1
        }
        {
         arriveTangent=0.0961
         arriveTangentWeight=0.0181
         interpMode=RCIM_Cubic
         leaveTangent=135.621
         leaveTangentWeight=0.2105
         tangentMode=RCTM_Break
         time=0.3535
         value=1
        }
        {
         arriveTangentWeight=0.0025
         interpMode=RCIM_Cubic
         leaveTangent=-18.5918
         leaveTangentWeight=0.208
         tangentMode=RCTM_Break
         time=0.3611
         value=1.7
        }
        {
         arriveTangent=-1.3624
         arriveTangentWeight=0.0383
         interpMode=RCIM_Cubic
         leaveTangent=-0.0605
         leaveTangentWeight=0.1696
         tangentMode=RCTM_Break
         time=0.4661
         value=1
        }
        {interpMode=RCIM_Cubic, leaveTangent=-138.3839, tangentMode=RCTM_Break, time=0.985, value=1}
        {arriveTangentWeight=0.005, interpMode=RCIM_Cubic, tangentMode=RCTM_Break, time=1, value=0}
       ]
      }
      YCurve {
       defaultValue=1
       keys [
        {arriveTangent=-85.7451, interpMode=RCIM_Cubic, leaveTangent=-11.9958, leaveTangentWeight=0.3283, tangentMode=RCTM_Break, time=0, value=1.85}
        {
         arriveTangent=-0.1408
         arriveTangentWeight=0.0067
         interpMode=RCIM_Cubic
         leaveTangent=0.1534
         leaveTangentWeight=0.3217
         tangentMode=RCTM_Break
         time=0.12
         value=1
        }
        {
         arriveTangent=0.094
         arriveTangentWeight=0.0523
         interpMode=RCIM_Cubic
         leaveTangent=150.2723
         leaveTangentWeight=0.0025
         tangentMode=RCTM_Break
         time=0.1768
         value=1
        }
        {
         arriveTangent=0.0001
         arriveTangentWeight=0.0548
         interpMode=RCIM_Cubic
         leaveTangent=-12.1025
         leaveTangentWeight=0.2669
         tangentMode=RCTM_Break
         time=0.1843
         value=1.85
        }
        {
         arriveTangent=-0.0955
         arriveTangentWeight=0.005
         interpMode=RCIM_Cubic
         leaveTangent=-0.0955
         leaveTangentWeight=0.2619
         tangentMode=RCTM_Break
         time=0.2994
         value=1
        }
        {
         arriveTangent=0.0961
         arriveTangentWeight=0.0514
         interpMode=RCIM_Cubic
         leaveTangent=135.621
         leaveTangentWeight=0.2105
         tangentMode=RCTM_Break
         time=0.3535
         value=1
        }
        {
         arriveTangentWeight=0.0025
         interpMode=RCIM_Cubic
         leaveTangent=-16.9812
         leaveTangentWeight=0.208
         tangentMode=RCTM_Break
         time=0.3611
         value=1.85
        }
        {
         arriveTangent=-1.3624
         arriveTangentWeight=0.005
         interpMode=RCIM_Cubic
         leaveTangent=-0.0605
         leaveTangentWeight=0.203
         tangentMode=RCTM_Break
         time=0.4761
         value=1
        }
        {interpMode=RCIM_Cubic, leaveTangent=-138.3839, tangentMode=RCTM_Break, time=0.985, value=1}
        {arriveTangentWeight=0.005, interpMode=RCIM_Cubic, tangentMode=RCTM_Break, time=1, value=0}
       ]
      }
     }}
    Scale Sprite Size Mode=::NewEnumerator3
   ) {module=ScaleSpriteSize003}
   &ScaleColor(
    Scale RGB=&VectorFromFloat(Value=&FloatFromCurve(
      FloatCurve=DataInterface() {propertyValues {Curve {keys [
          {interpMode=RCIM_Cubic, leaveTangent=-22.1259, tangentMode=RCTM_Break, time=0.0633, value=3}
          {interpMode=RCIM_Cubic, tangentMode=RCTM_Break, time=0.11, value=2.1708}
          <at 0.5 1.5>
         ]}}}
      Scale Curve=&FloatFromCurve(
       FloatCurve=DataInterface() {propertyValues {Curve {keys [
           {interpMode=RCIM_Cubic, leaveTangent=-10.6544, tangentMode=RCTM_Break, time=0, value=2}
           {interpMode=RCIM_Cubic, tangentMode=RCTM_Break, time=0.1, value=1}
           {interpMode=RCIM_Cubic, tangentMode=RCTM_Break, time=0.1768, value=1}
           {interpMode=RCIM_Cubic, leaveTangent=-10.4556, tangentMode=RCTM_Break, time=0.1843, value=2}
           {interpMode=RCIM_Cubic, tangentMode=RCTM_Break, time=0.2894, value=1}
           {interpMode=RCIM_Cubic, tangentMode=RCTM_Break, time=0.3535, value=1}
           {interpMode=RCIM_Cubic, leaveTangent=-9.6474, tangentMode=RCTM_Break, time=0.3611, value=2}
           {interpMode=RCIM_Cubic, tangentMode=RCTM_Break, time=0.4661, value=1}
           {interpMode=RCIM_Cubic, tangentMode=RCTM_Break, time=1, value=1}
          ]}}}
       CurveIndex=$System.NormalizedLoopAge
      )
     ))
    ScaleA=true
    ScaleRGB=true
   )
  ]
  rendererClasses [&NiagaraSpriteRendererProperties]
  renderers [{rendererClass=&NiagaraSpriteRendererProperties}]
 }
 Flare {bEnabled=false, rendererClasses [], renderers []}
 Flare001 {bEnabled=false, rendererClasses [], renderers []}
 Flare002 {rendererClasses [], renderers []}
 Flare003 {rendererClasses [], renderers []}
 Light {rendererClasses [], renderers []}
 Spark_Large_02 {rendererClasses [], renderers []}
 Spark_Large_02001 {bEnabled=false, rendererClasses [], renderers []}
 Spark_Large_03 {rendererClasses [], renderers []}
 Spark_Large_03001 {bEnabled=false, rendererClasses [], renderers []}
 Spark_Large_04 {rendererClasses [], renderers []}
 Spark_Large_04001 {bEnabled=false, rendererClasses [], renderers []}
 Spark_Small {bEnabled=false, rendererClasses [], renderers []}
 Spark_Small001 {rendererClasses [], renderers []}
}
summary {
 systemName=NS_Emote_Wingbath_Spark_FX
 emitters [
  {bEnabled=false, emitterName=Spark_Small, rendererClasses []}
  {bEnabled=false, emitterName=Flare, rendererClasses []}
  {bEnabled=false, emitterName=Flame002, rendererClasses [&NiagaraSpriteRendererProperties]}
  {bEnabled=false, emitterName=Spark_Large_02001, rendererClasses []}
  {bEnabled=false, emitterName=Spark_Large_03001, rendererClasses []}
  {bEnabled=false, emitterName=Spark_Large_04001, rendererClasses []}
  {bEnabled=false, emitterName=Flare001, rendererClasses []}
  {bEnabled=false, emitterName=Flame, rendererClasses [&NiagaraSpriteRendererProperties]}
  {emitterName=Spark_Small001, rendererClasses []}
  {emitterName=Flare002, rendererClasses []}
  {emitterName=Spark_Large_02, rendererClasses []}
  {emitterName=Spark_Large_03, rendererClasses []}
  {emitterName=Spark_Large_04, rendererClasses []}
  {emitterName=Flare003, rendererClasses []}
  {emitterName=Light, rendererClasses []}
 ]
 userVariables [
  User.bIsFrontEnd {type=NiagaraBool}
  User.FrontEndScale {type=NiagaraFloat, defaultValue=0.6, description=""}
 ]
}
